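# Physics asset for the Dragon Mask pickaxe, a two-handed melee cosmetic, defining collision bodies and joints for its dangling hair and ropes.
# The handle body combines a capsule shaft with a box head and anchors all dynamic chains.
# Three capsule hair bodies attach directly to the handle with tight swing and twist limits.
# Left and right three-link rope chains use double capsules ending in spheres, with wider swing and locked twist.
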PhysicsAsset /BRCosmetics/Weapons/FORT_Melee/Pickaxe_DragonMask/Meshes/Pickaxe_Dragon_Mask_Physics
bodies {
 Melee_TwoHanded_Handle [
  {shapeType=Capsule, length=88.9389, radius=2.6944, center=<0.8672 -0.9021 25.7883>, rotation=<rot -0.0543 -89.9991 -0.0259>}
  {shapeType=Box, extentX=9.9399, extentY=23.1314, extentZ=22.6408, center=<0.6395 -0.8479 76.8859>}
 ]
 dyn_l_hair [{shapeType=Capsule, length=9.9136, radius=4.8162, center=<0.6167 0.1144 1.2357>, rotation=<rot 5.9939 -89.7031 -93.475>}]
 dyn_l_rope_01 [
  {shapeType=Capsule, length=2.4525, radius=0.7556, center=<0.9762 0.0192 -0.0161>, rotation=<rot 0 -92.7671 -90.5097>}
  {shapeType=Capsule, length=3.1463, radius=0.7556, center=<3.6555 0.3729 -0.7175>, rotation=<rot 0 -92.7671 -60.5127>}
 ]
 dyn_l_rope_02 [
  {shapeType=Capsule, length=2.8959, radius=0.7441, center=<0.6693 -0.0298 0.1203>, rotation=<rot 0 -89.2615 -91.1273>}
  {shapeType=Capsule, length=2.8586, radius=0.7441, center=<3.4597 -0.1229 -0.2603>, rotation=<rot 0 -89.2615 -81.1235>}
 ]
 dyn_l_rope_03 [{shapeType=Sphere, radius=2.5104, center=<1.9229 0.0001 -0.224>}]
 dyn_m_hair [{shapeType=Capsule, length=14.0564, radius=5.3553, center=<0.8397 -0.0003 0.3311>, rotation=<rot 0.0002 90.0023 -90.0001>}]
 dyn_r_hair [{shapeType=Capsule, length=9.9136, radius=4.8162, center=<0.9337 0.3735 0.1827>, rotation=<rot -5.9916 -90.2937 -93.4743>}]
 dyn_r_rope_01 [
  {shapeType=Capsule, length=3.5305, radius=0.7556, center=<0.685 -0.0628 -0.1768>, rotation=<rot 0 92.7671 -80.5097>}
  {shapeType=Capsule, length=2.9695, radius=0.7556, center=<3.5918 -0.4056 0.9449>, rotation=<rot 0 92.7671 -55.5102>}
 ]
 dyn_r_rope_03 [{shapeType=Sphere, radius=2.6708, center=<1.9213 0.0001 0.0791>}]
 dyn_r_rope_0_2 [
  {shapeType=Capsule, length=2.4892, radius=0.7441, center=<0.86 0.0734 0.1198>, rotation=<rot 0 89.2615 -91.1277>}
  {shapeType=Capsule, length=2.9689, radius=0.7441, center=<3.6823 0.1659 0.4661>, rotation=<rot 0 89.2615 -76.1263>}
 ]
}
constraints [
 {bone1Name=dyn_m_hair, bone2Name=Melee_TwoHanded_Handle, swing1LimitDegrees=20, swing2LimitDegrees=20, twistLimitDegrees=10}
 {bone1Name=dyn_l_hair, bone2Name=Melee_TwoHanded_Handle, swing1LimitDegrees=20, swing2LimitDegrees=20, twistLimitDegrees=10}
 {bone1Name=dyn_r_hair, bone2Name=Melee_TwoHanded_Handle, swing1LimitDegrees=20, swing2LimitDegrees=20, twistLimitDegrees=10}
 {
  bone1Name=dyn_l_rope_01
  bone2Name=Melee_TwoHanded_Handle
  swing1LimitDegrees=45
  swing2LimitDegrees=30
  twistLimitDegrees=45
  twistMotion=Locked
 }
 {bone1Name=dyn_l_rope_02, bone2Name=dyn_l_rope_01, swing1LimitDegrees=45, swing2LimitDegrees=30, twistLimitDegrees=45, twistMotion=Locked}
 {bone1Name=dyn_l_rope_03, bone2Name=dyn_l_rope_02, swing1LimitDegrees=45, swing2LimitDegrees=30, twistLimitDegrees=45, twistMotion=Locked}
 {
  bone1Name=dyn_r_rope_01
  bone2Name=Melee_TwoHanded_Handle
  swing1LimitDegrees=45
  swing2LimitDegrees=30
  twistLimitDegrees=45
  twistMotion=Locked
 }
 {bone1Name=dyn_r_rope_0_2, bone2Name=dyn_r_rope_01, swing1LimitDegrees=45, swing2LimitDegrees=30, twistLimitDegrees=45, twistMotion=Locked}
 {bone1Name=dyn_r_rope_03, bone2Name=dyn_r_rope_0_2, swing1LimitDegrees=45, swing2LimitDegrees=30, twistLimitDegrees=45, twistMotion=Locked}
]
